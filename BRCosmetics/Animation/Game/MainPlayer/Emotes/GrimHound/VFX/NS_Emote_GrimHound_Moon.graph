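# Niagara effect for the GrimHound emote that shows a glowing cyan moon with an ink circle, swirls, ring glow and smoke rings.
# Timed bursts, delays and loop-age curves sequence sprite and mesh emitters over roughly seven seconds.
# The moon mesh animates through dynamic material parameters and scale and color curves.
# Camera-angle dot products adjust glow and smoke, and spawn positions shift in the front end versus gameplay.
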NiagaraSystem /BRCosmetics/Animation/Game/MainPlayer/Emotes/GrimHound/VFX/NS_Emote_GrimHound_Moon
dependencies {
 usedDataInterfaces [
  &NiagaraDataInterfaceCurve
  &NiagaraDataInterfaceColorCurve
  &NiagaraDataInterfaceCamera
  &NiagaraDataInterfaceVectorCurve
 ]
 usedDynamicInputs []
 usedModules [&SystemState]
}
emitters {
 Glow002 {
  emitterUpdateScript [
   &EmitterState(Life Cycle Mode=::NewEnumerator1, Loop Behavior=::NewEnumerator1, Loop Duration=7, UseLoopDelay=false)
   &SpawnBurst_Instantaneous(Spawn Time=0.63, Use Spawn Probability=false, UseLoopCountLimit=false) {module=SpawnBurst_Instantaneous001}
   &SpawnBurst_Instantaneous(Spawn Group=1, Spawn Time=0.63, Use Spawn Probability=false, UseLoopCountLimit=false)
  ]
  particleSpawnScript [
   /SetVariables(Particles.VisibilityTag=&SpawnGroup()) {isSetParameters=true}
   &InitializeParticle(
    Color=&MergeHelper_ColorSwitch(
     0=<rgba 0.4428 0.9156 0.8718 0.5>
     1=<rgba 0.2796 0.5781 0.5505 0.75>
     2=?
     3=?
     Index=&SpawnGroup()
     Num=::NewEnumerator0
    )
    Lifetime=7
    Position Offset=&FN_FrontEndVector3Modifier(Gameplay=<0 0 0>, Front End=<-10 0 0>)
    Sprite Size Mode=::NewEnumerator3
    Uniform Sprite Size=&MergeHelper_FloatSwitch(0=290, 1=162, 2=?, 3=?, Index=&SpawnGroup(), Num=::NewEnumerator0)
    UsePositionOffset=true
    Write Material Random=true
   )
   &SpriteFacingAndAlignment(Sprite Facing=<0 1 0>, WriteAlignment=false, WriteFacing=true)
  ]
  particleUpdateScript [
   &ParticleState()
   &ScaleSpriteSize(
    Scale Factor=&Multiply_Vector2D(
     A=&Vector2DFromFloat(Value=&Multiply_Float(
       A=&FloatFromCurve(FloatCurve=DataInterface() {propertyValues {Curve {
           defaultValue=1
           keys [
            {interpMode=RCIM_Cubic, leaveTangent=10.6116, tangentMode=RCTM_Break, time=0.0014, value=1.0218}
            {arriveTangent=-4.462, interpMode=RCIM_Cubic, leaveTangent=-0.0155, tangentMode=RCTM_Break, time=0.0281, value=1.0063}
            {arriveTangent=0.7744, interpMode=RCIM_Cubic, leaveTangent=0.7744, tangentMode=RCTM_User, time=0.2025, value=1.018}
            {arriveTangent=3.1993, interpMode=RCIM_Cubic, leaveTangent=30.0615, tangentMode=RCTM_Break, time=0.4037, value=1.4615}
            {arriveTangent=-5.824, interpMode=RCIM_Cubic, tangentMode=RCTM_Break, time=0.4164, value=1.4541}
           ]
          }}})
       B=1
      ))
     B=&MakeVector2D(
      X=&MergeHelper_FloatSwitch(
       0=&RemapRange(
        Input Value To Remap=&Abs_Float(Float:NiagaraFloat=&DotProduct(
          Vector A=&ReturnCameraVector(Camera Vectors=::NewEnumerator4, Coordinate Space=::Simulation, Origin Position=$Particles.Position)
          Vector B=$Particles.SpriteFacing
         ))
        Input Min=0.2
        Input Max=0.9
        Output Min=0.43
        Output Max=1
       )
       1=1
       2=?
       3=?
       Index=&SpawnGroup()
       Num=::NewEnumerator0
      )
      Y=1
     )
    )
    Scale Sprite Size Mode=::NewEnumerator2
   ) {module=ScaleSpriteSize001}
   &ScaleColor(
    Scale Alpha=&FloatFromCurve(
     FloatCurve=DataInterface() {propertyValues {Curve {
        defaultValue=1
        keys [
         {interpMode=RCIM_Cubic, tangentMode=RCTM_User, time=0, value=0}
         {arriveTangentWeight=0.071, interpMode=RCIM_Cubic, leaveTangentWeight=0.23, tangentMode=RCTM_User, time=0.202, value=0.98}
         {arriveTangentWeight=0.2481, interpMode=RCIM_Cubic, leaveTangentWeight=0.0544, tangentMode=RCTM_User, time=0.8231, value=1}
         {tangentMode=RCTM_User, time=1, value=0}
        ]
       }}}
     Scale Curve=0.75
    )
    Scale RGB=&MakeVectorFromLinearColorRGB(Color=&ColorFromCurve(ColorCurve=DataInterface() {propertyValues {
        AlphaCurve {keys [<at 0.0263 0.5>,<at 0.104 1 RCIM_Cubic>]}
        BlueCurve {keys [<at 0.0213 0.5437 RCIM_Cubic>,<at 0.0992 0.6073>,<at 0.3899 1>,<at 0.4069 2>,<at 0.4239 1>,<at 1 1 RCIM_Cubic>]}
        GreenCurve {keys [<at 0.0213 0.3928 RCIM_Cubic>,<at 0.0992 0.4287>,<at 0.3899 1>,<at 0.4069 2>,<at 0.4239 1>,<at 1 1 RCIM_Cubic>]}
        RedCurve {keys [<at 0.0213 0.0375 RCIM_Cubic>,<at 0.0992 0.0572>,<at 0.3899 1>,<at 0.4069 2>,<at 0.4239 1>,<at 1 1 RCIM_Cubic>]}
       }}))
    ScaleA=true
    ScaleRGB=true
   )
   /SetVariables(Particles.Linear Color=&MakeLinearColorFromVectorAndFloat(
     "Vector (RGB)"=&MakeVectorFromLinearColorRGB(Color=<rgba 0.7254 1.5 1.4283 1>)
     "Float (Alpha)"=&Multiply_Float(
      A=&FloatFromCurve(FloatCurve=DataInterface() {propertyValues {Curve {
          defaultValue=1
          keys [
           <at -0.0093 0 RCIM_Cubic>
           {arriveTangent=1.9634, interpMode=RCIM_Cubic, leaveTangent=1.9634, time=0.311, value=0.8399}
           {arriveTangent=-0.5838, interpMode=RCIM_Cubic, leaveTangent=-0.5838, tangentMode=RCTM_User, time=0.5, value=1}
           {interpMode=RCIM_Cubic, tangentMode=RCTM_User, time=0.9898, value=-0.0196}
          ]
         }}})
      B=&RemapRange(
       Input Value To Remap=&Abs_Float(Float:NiagaraFloat=&DotProduct(
         Vector A=&ReturnCameraVector(Camera Vectors=::NewEnumerator4, Coordinate Space=::Simulation, Origin Position=$Particles.Position)
         Vector B=$Particles.SpriteFacing
        ))
       Input Min=0.7
       Input Max=0.9
       Output Min=0
       Output Max=0.1
      )
     )
    )) {isSetParameters=true}
  ]
  rendererClasses [&NiagaraSpriteRendererProperties]
  renderers [
   {rendererClass=&NiagaraSpriteRendererProperties}
   {rendererClass=&NiagaraSpriteRendererProperties, rendererIndex=1}
  ]
 }
 Ink_Circle {
  emitterUpdateScript [
   &EmitterState(Life Cycle Mode=::NewEnumerator1, Loop Behavior=::NewEnumerator1, Loop Duration=3, UseLoopDelay=false)
   &SpawnBurst_Instantaneous(Use Spawn Probability=false, UseLoopCountLimit=false)
   &SpawnBurst_Instantaneous(Spawn Group=1, Use Spawn Probability=false, UseLoopCountLimit=false) {module=SpawnBurst_Instantaneous001}
  ]
  particleSpawnScript [
   /SetVariables(Particles.VisibilityTag=&SpawnGroup()) {isSetParameters=true}
   &InitializeParticle(
    Color=&MergeHelper_ColorSwitch(
     0=<rgba 0 0.025 0.026 1>
     1=<rgba 0.2344 0.9679 1 0.6>
     2=<rgba 1 1 1 0>
     3=?
     Index=&SpawnGroup()
     Num=::NewEnumerator1
    )
    Lifetime=&MergeHelper_FloatSwitch(
     0=2.6
     1=2.6
     2=&RandomRangeFloat(Minimum=0.3, Maximum=0.7)
     3=?
     Index=&SpawnGroup()
     Num=::NewEnumerator1
    )
    Position Offset=&FN_FrontEndVector3Modifier(Gameplay=<0 0 0>, Front End=<-10 0 0>)
    Sprite Size Mode=::NewEnumerator3
    Uniform Sprite Size=118.5
    UsePositionOffset=true
    Write Material Random=true
   )
   &InitialMeshOrientation(
    Mesh Axis To Orient=<0 1 0>
    Mesh Orientation Mode=::NewEnumerator2
    Orientation Vector=<0 0 1>
    Rotation=&RandomRangeVector(Minimum=<0 0 0>, Maximum=<0 0 2>)
    Use Rotation Vector=true
   )
   &SpriteFacingAndAlignment(Sprite Facing=<0 1 0>, WriteAlignment=false, WriteFacing=true)
  ]
  particleUpdateScript [
   &ParticleState()
   &SpriteRotationRate(Rotation Rate=&FloatFromCurve(FloatCurve=DataInterface() {propertyValues {Curve {keys [<at 0 325>,<at 1 600>]}}}))
   /SetVariables(Particles.Sprite Dynamic Param:Vector4f=&MakeVector4(
     X=&MergeHelper_FloatSwitch(
      0=&FloatFromCurve(FloatCurve=DataInterface() {propertyValues {Curve {keys [
           {arriveTangent=-6.9561, interpMode=RCIM_Cubic, leaveTangent=-6.9561, tangentMode=RCTM_User, time=0, value=2}
           {arriveTangent=-14.7608, interpMode=RCIM_Cubic, leaveTangent=-14.7608, tangentMode=RCTM_User, time=0.0938, value=1.0737}
           {interpMode=RCIM_Cubic, tangentMode=RCTM_User, time=0.2712, value=-0.0403}
           {arriveTangent=0.3328, interpMode=RCIM_Cubic, leaveTangent=0.3328, tangentMode=RCTM_User, time=0.4436, value=0}
           {arriveTangent=1.8188, interpMode=RCIM_Cubic, leaveTangent=2.5173, tangentMode=RCTM_Break, time=0.7118, value=0.5768}
           {arriveTangent=9.1725, interpMode=RCIM_Cubic, leaveTangent=9.1725, tangentMode=RCTM_User, time=0.961, value=2.0427}
          ]}}})
      1=&FloatFromCurve(FloatCurve=DataInterface() {propertyValues {Curve {keys [
           {arriveTangent=-6.9561, interpMode=RCIM_Cubic, leaveTangent=-6.9561, tangentMode=RCTM_User, time=0, value=2}
           {arriveTangent=-14.7608, interpMode=RCIM_Cubic, leaveTangent=-14.7608, tangentMode=RCTM_User, time=0.0937, value=1.0737}
           {interpMode=RCIM_Cubic, tangentMode=RCTM_User, time=0.213, value=0}
           {arriveTangent=0.7046, interpMode=RCIM_Cubic, leaveTangent=0.7046, time=0.3883, value=0}
           {arriveTangent=1.8108, interpMode=RCIM_Cubic, leaveTangent=-4.6777, tangentMode=RCTM_Break, time=0.651, value=0.3086}
           {arriveTangent=13.9061, interpMode=RCIM_Cubic, leaveTangent=9.5155, tangentMode=RCTM_Break, time=0.7206, value=0.3184}
           {arriveTangent=-0.1502, interpMode=RCIM_Cubic, leaveTangent=-0.1502, tangentMode=RCTM_User, time=0.9777, value=2}
          ]}}})
      2=?
      3=?
      Index=&SpawnGroup()
      Num=::NewEnumerator0
     )
     Y=1
     Z=0
     W=0
    )) {isSetParameters=true}
   &ScaleColor(
    Scale Alpha=&MergeHelper_FloatSwitch(
     0=&FloatFromCurve(FloatCurve=DataInterface() {propertyValues {Curve {keys [
          {arriveTangent=0.3355, interpMode=RCIM_Cubic, leaveTangent=0.3355, tangentMode=RCTM_User, time=0.038, value=0}
          {arriveTangent=1.2948, interpMode=RCIM_Cubic, leaveTangent=1.2948, time=0.5697, value=0.3447}
          {interpMode=RCIM_Cubic, tangentMode=RCTM_User, time=0.8103, value=1}
         ]}}})
     1=&FloatFromCurve(FloatCurve=DataInterface() {propertyValues {Curve {keys [
          {arriveTangent=0.3355, interpMode=RCIM_Cubic, leaveTangent=0.3355, tangentMode=RCTM_User, time=0.038, value=0}
          {arriveTangent=1.2948, interpMode=RCIM_Cubic, leaveTangent=1.2948, time=0.5697, value=0.3447}
          {interpMode=RCIM_Cubic, tangentMode=RCTM_User, time=0.8103, value=1}
         ]}}})
     2=&Multiply_Float(
      A=&FloatFromCurve(
       FloatCurve=DataInterface() {propertyValues {Curve {
          defaultValue=1
          keys [
           {interpMode=RCIM_Cubic, tangentMode=RCTM_User, time=0, value=0}
           {interpMode=RCIM_Cubic, tangentMode=RCTM_User, time=0.1617, value=0.6549}
           {tangentMode=RCTM_User, time=1, value=0}
          ]
         }}}
       Scale Curve=0.4
      )
      B=&FloatFromCurve(
       FloatCurve=DataInterface() {propertyValues {Curve {keys [
           {arriveTangent=0.5594, interpMode=RCIM_Cubic, leaveTangent=0.5594, tangentMode=RCTM_User, time=0.0302, value=0.0236}
           {arriveTangent=2.4976, interpMode=RCIM_Cubic, leaveTangent=2.4976, tangentMode=RCTM_User, time=0.3927, value=0.5587}
           {interpMode=RCIM_Cubic, tangentMode=RCTM_User, time=0.6307, value=1}
          ]}}}
       CurveIndex=$Emitter.NormalizedLoopAge
      )
     )
     3=?
     Index=&SpawnGroup()
     Num=::NewEnumerator1
    )
    ScaleA=true
    ScaleRGB=true
   ) {module=ScaleColor001}
  ]
  rendererClasses [&NiagaraSpriteRendererProperties]
  renderers [
   {rendererClass=&NiagaraSpriteRendererProperties}
   {rendererClass=&NiagaraSpriteRendererProperties, rendererIndex=1}
  ]
 }
 Mesh_Swirls {
  emitterUpdateScript [
   &EmitterState(Life Cycle Mode=::NewEnumerator1, Loop Behavior=::NewEnumerator1, Loop Duration=3.7, UseLoopDelay=false)
   &SpawnRate(
    SpawnRate=&FloatFromCurve(
     FloatCurve=DataInterface() {propertyValues {Curve {keys [
         {interpMode=RCIM_Cubic, tangentMode=RCTM_User, time=0.6994, value=1.3416}
         {arriveTangent=-6.0816, interpMode=RCIM_Cubic, leaveTangent=-6.0816, time=0.7096, value=0.9944}
         {interpMode=RCIM_Cubic, tangentMode=RCTM_User, time=0.92, value=0}
        ]}}}
     CurveIndex=$Emitter.NormalizedLoopAge
     Scale Curve=17
    )
    Use Spawn Probability=false
   ) {module=SpawnRate001}
  ]
  particleSpawnScript [
   &InitializeParticle(
    Color Mode=::NewEnumerator0
    Lifetime Max=0.7
    Lifetime Min=0.3
    Lifetime Mode=::NewEnumerator1
    Mesh Scale Max=<1 1 1.7>
    Mesh Scale Mode=::NewEnumerator2
    Position Offset=&FN_FrontEndVector3Modifier(Gameplay=<0 0 0>, Front End=<-10 0 0>)
    UsePositionOffset=true
    Write Material Random=true
   )
   &InitialMeshOrientation(
    Mesh Axis To Orient=<0 1 0>
    Mesh Orientation Mode=::NewEnumerator2
    Orientation Vector=<0 0 1>
    Rotation=&RandomRangeVector(Minimum=<0 0 0>, Maximum=<0 0 2>)
    Use Rotation Vector=true
   )
  ]
  particleUpdateScript [
   &ParticleState()
   &UpdateMeshOrientation(
    Rotation Rate=&Multiply_Float(
     A=&Lerp_Float(
      A=&RandomRangeFloat(Minimum=-0.7, Maximum=-1)
      B=&RandomRangeFloat(Minimum=0.7, Maximum=1)
      Alpha=&RandomRangeFloat(Minimum=0, Maximum=1)
     )
     B=&FloatFromCurve(
      FloatCurve=DataInterface() {propertyValues {Curve {
         defaultValue=1.8
         keys [
          {interpMode=RCIM_Cubic, leaveTangentWeight=0.143, tangentMode=RCTM_User, time=0.621, value=1.1333}
          {arriveTangent=11.5171, interpMode=RCIM_Cubic, leaveTangent=11.5171, tangentMode=RCTM_User, time=0.8081, value=2.1661}
          {arriveTangent=21.816, interpMode=RCIM_Cubic, leaveTangent=127.4518, tangentMode=RCTM_Break, time=0.8489, value=2.9409}
          {arriveTangent=-16.9051, interpMode=RCIM_Cubic, leaveTangent=-16.9051, time=0.8987, value=4.9595}
          {arriveTangent=-175.3616, interpMode=RCIM_Cubic, leaveTangent=-175.3616, tangentMode=RCTM_User, time=0.9048, value=1.996}
         ]
        }}}
      CurveIndex=$Emitter.NormalizedLoopAge
     )
    )
    Rotation Vector=<0 0 1>
   )
   &ScaleMeshSize(Scale Factor=&VectorFromFloat(Value=&FloatFromCurve(
      FloatCurve=DataInterface() {propertyValues {Curve {
         defaultValue=1
         keys [{interpMode=RCIM_Cubic, tangentMode=RCTM_User, time=0, value=1}, {interpMode=RCIM_Cubic, tangentMode=RCTM_User, time=1, value=1.5}]
        }}}
      Scale Curve=&FloatFromCurve(
       FloatCurve=DataInterface() {propertyValues {Curve {
          defaultValue=1
          keys [
           {arriveTangent=3.3886, interpMode=RCIM_Cubic, leaveTangent=3.3886, tangentMode=RCTM_User, time=0.4154, value=1.4}
           {interpMode=RCIM_Cubic, tangentMode=RCTM_User, time=1, value=2}
          ]
         }}}
       CurveIndex=$Emitter.NormalizedLoopAge
       Scale Curve=&RandomRangeFloat(Minimum=0.7, Maximum=0.8)
      )
     )))
   &ScaleColor(
    Scale Alpha=&Multiply_Float(
     A=&FloatFromCurve(
      FloatCurve=DataInterface() {propertyValues {Curve {
         defaultValue=1
         keys [
          {interpMode=RCIM_Cubic, tangentMode=RCTM_User, time=0, value=0}
          {interpMode=RCIM_Cubic, tangentMode=RCTM_User, time=0.1617, value=0.6549}
          {tangentMode=RCTM_User, time=1, value=0}
         ]
        }}}
      Scale Curve=0.4
     )
     B=&FloatFromCurve(
      FloatCurve=DataInterface() {propertyValues {Curve {keys [
          {arriveTangent=0.5594, interpMode=RCIM_Cubic, leaveTangent=0.5594, tangentMode=RCTM_User, time=0.0302, value=0.0236}
          {arriveTangent=2.4976, interpMode=RCIM_Cubic, leaveTangent=2.4976, tangentMode=RCTM_User, time=0.3927, value=0.5587}
          {interpMode=RCIM_Cubic, tangentMode=RCTM_User, time=0.6307, value=1}
         ]}}}
      CurveIndex=$Emitter.NormalizedLoopAge
     )
    )
    ScaleA=true
    ScaleRGB=true
   )
  ]
  rendererClasses [&NiagaraMeshRendererProperties]
  renderers [{rendererClass=&NiagaraMeshRendererProperties, rendererIndex=0}]
 }
 Moon001 {
  emitterUpdateScript [
   &EmitterState(Life Cycle Mode=::NewEnumerator1, Loop Behavior=::NewEnumerator1, Loop Duration=6.6, UseLoopDelay=false)
   &SpawnBurst_Instantaneous(Spawn Time=0.63, Use Spawn Probability=false, UseLoopCountLimit=false)
   /SetVariables(Emitter.SphereMaskHardness=0.6) {isSetParameters=true}
  ]
  particleSpawnScript [
   /SetVariables(Particles.VisibilityTag=&SpawnGroup()) {isSetParameters=true}
   &InitializeParticle(
    Color=<rgba 1.9496 5 4.9523 0.5>
    Lifetime=$Emitter.CurrentLoopDuration
    Mesh Scale=<0.45 1.05 1.05>
    Mesh Scale Mode=::NewEnumerator4
    Position Offset=&FN_FrontEndVector3Modifier(Gameplay=<0 0 0>, Front End=<-10 0 0>)
    UsePositionOffset=true
    Write Material Random=true
   )
   &InitialMeshOrientation(Mesh Orientation Mode=::NewEnumerator2, Orientation Vector=<0 1 0>, Use Rotation Vector=false)
  ]
  particleUpdateScript [
   &ParticleState()
   &ScaleMeshSize(Scale Factor=&VectorFromFloat(Value=&FloatFromCurve(FloatCurve=DataInterface() {propertyValues {Curve {
         defaultValue=1
         keys [
          {interpMode=RCIM_Cubic, leaveTangent=34.8392, tangentMode=RCTM_Break, time=0.0043, value=1.0167}
          {arriveTangent=-4.462, interpMode=RCIM_Cubic, leaveTangent=-0.0155, tangentMode=RCTM_Break, time=0.0293, value=1.0011}
          {arriveTangentWeight=0.0299, interpMode=RCIM_Cubic, leaveTangentWeight=0.0229, tangentMode=RCTM_User, time=0.2, value=1}
          {arriveTangent=2.3505, interpMode=RCIM_Cubic, leaveTangent=2.3505, tangentMode=RCTM_User, time=0.3126, value=1.1104}
          {arriveTangent=4.3732, interpMode=RCIM_Cubic, leaveTangent=37.4415, tangentMode=RCTM_Break, time=0.4333, value=1.4869}
          {arriveTangent=-5.824, interpMode=RCIM_Cubic, tangentMode=RCTM_Break, time=0.4427, value=1.4795}
         ]
        }}}))) {module=ScaleMeshSize001}
   &UpdateMeshOrientation(
    Facing Direction=&MergeHelper_VectorSwitch(0=<1 0 0>, 1=<0 0 1>, 2=?, 3=?, Index=&SpawnGroup(), Num=::NewEnumerator0)
    Facing Interpolation=false
    Orientation Method=::NewEnumerator1
    Override Quat=false
    TransformFacing=false
    TransformUp=false
    Up Direction:Vector3f=&MergeHelper_VectorSwitch(0=<0 0 1>, 1=<1 0 0>, 2=?, 3=?, Index=&SpawnGroup(), Num=::NewEnumerator0)
   )
   &SolveForcesAndVelocity(Clamp Velocity=false, Limit Acceleration=false)
   &DynamicMaterialParameters(
    Index 0 Param 1=0.6
    Index 1 Param 1=&FloatFromCurve(FloatCurve=DataInterface() {propertyValues {Curve {
        defaultValue=1
        keys [
         {arriveTangent=4.9288, interpMode=RCIM_Cubic, leaveTangent=4.9288, tangentMode=RCTM_User, time=0, value=0.0298}
         {arriveTangent=1.4561, interpMode=RCIM_Cubic, leaveTangent=1.4561, time=0.0968, value=0.3064}
         {arriveTangent=4.4911, interpMode=RCIM_Cubic, leaveTangent=4.4911, tangentMode=RCTM_User, time=0.4647, value=0.7066}
         {arriveTangent=4.0766, interpMode=RCIM_Cubic, leaveTangent=4.0766, tangentMode=RCTM_User, time=0.511, value=0.99}
        ]
       }}})
    Index 1 Param 2=&FloatFromCurve(FloatCurve=DataInterface() {propertyValues {Curve {keys [
         {arriveTangent=4.2057, interpMode=RCIM_Cubic, leaveTangent=4.2057, tangentMode=RCTM_User, time=0, value=0}
         {arriveTangent=2.0743, interpMode=RCIM_Cubic, leaveTangent=10.5882, tangentMode=RCTM_Break, time=0.3957, value=0.6542}
         {arriveTangent=-1.1976, interpMode=RCIM_Cubic, leaveTangent=-0.0385, tangentMode=RCTM_Break, time=0.5067, value=1}
         <at 0.9303 1>
         <at 1 0>
        ]}}})
    Index 1 Param 3=&FloatFromCurve(FloatCurve=DataInterface() {propertyValues {Curve {
        defaultValue=1
        keys [
         {interpMode=RCIM_Cubic, tangentMode=RCTM_User, time=0.0008, value=-0.1047}
         {arriveTangent=8.4811, interpMode=RCIM_Cubic, leaveTangent=8.4811, time=0.1319, value=0.6717}
         {arriveTangent=2.6356, interpMode=RCIM_Cubic, leaveTangent=2.6356, tangentMode=RCTM_User, time=0.2544, value=2.0462}
         {arriveTangent=-2.7443, interpMode=RCIM_Cubic, leaveTangent=-2.7443, time=0.8171, value=2.0697}
         {tangentMode=RCTM_User, time=1, value=0}
        ]
       }}})
    Index 1 Param 4=&FloatFromCurve(FloatCurve=DataInterface() {propertyValues {Curve {keys [<at 0.5062 1.5>,<at 0.5073 -1>]}}})
    Module.Index0Float0Write=true
    Module.Index0Float1Write=true
    Module.Index0Float2Write=true
    Module.Index0Float3Write=true
    Module.Index1Float0Write=true
    Module.Index1Float1Write=true
    Module.Index1Float2Write=true
    Module.Index1Float3Write=true
    Write Parameter Index 1=::NewEnumerator1
   ) {module=DynamicMaterialParameters003}
   &ScaleColor(
    Scale Alpha=&FloatFromCurve(
     FloatCurve=DataInterface() {propertyValues {Curve {
        defaultValue=1
        keys [
         {arriveTangent=121.2426, interpMode=RCIM_Cubic, leaveTangent=121.2426, tangentMode=RCTM_User, time=0, value=0}
         {
          arriveTangent=0.3077
          arriveTangentWeight=0.0199
          interpMode=RCIM_Cubic
          leaveTangent=0.3077
          leaveTangentWeight=0.2613
          tangentMode=RCTM_User
          time=0.0205
          value=1.3966
         }
         {arriveTangent=0.6186, interpMode=RCIM_Cubic, leaveTangent=0.6186, tangentMode=RCTM_User, time=0.1576, value=1.0645}
         {
          arriveTangent=-0.523
          arriveTangentWeight=0.3814
          interpMode=RCIM_Cubic
          leaveTangent=-0.523
          leaveTangentWeight=0.0762
          tangentMode=RCTM_User
          time=0.75
          value=1
         }
         {arriveTangent=-4, interpMode=RCIM_Cubic, leaveTangent=-4, time=0.8958, value=0.2748}
         {tangentMode=RCTM_User, time=1, value=0}
        ]
       }}}
     Scale Curve=0.92
    )
    Scale RGB=&MakeVectorFromLinearColorRGB(Color=&ColorFromCurve(ColorCurve=DataInterface() {propertyValues {
        AlphaCurve {keys []}
        BlueCurve {keys [<at 0 0.9809>,<at 0.058 1>,<at 0.1332 0.9601>,<at 1 0.374 RCIM_Cubic>]}
        GreenCurve {keys [<at 0 1>,<at 0.058 0.975>,<at 0.1332 1>,<at 1 0.374 RCIM_Cubic>]}
        RedCurve {keys [<at 0 0.6295>,<at 0.058 0.9739>,<at 0.1332 0.8525>,<at 1 0.374 RCIM_Cubic>]}
       }}))
    ScaleA=true
    ScaleRGB=true
   ) {enabled=false}
   &ScaleColor(
    Scale Alpha=&FloatFromCurve(
     FloatCurve=DataInterface() {propertyValues {Curve {
        defaultValue=1
        keys [
         {arriveTangent=121.2426, interpMode=RCIM_Cubic, leaveTangent=121.2426, tangentMode=RCTM_User, time=0, value=0}
         {
          arriveTangent=0.3077
          arriveTangentWeight=0.0199
          interpMode=RCIM_Cubic
          leaveTangent=0.3077
          leaveTangentWeight=0.2613
          tangentMode=RCTM_User
          time=0.0125
          value=2.0376
         }
         {arriveTangent=-6.706, interpMode=RCIM_Cubic, leaveTangent=-6.706, time=0.0931, value=1.3139}
         {arriveTangent=-5.1416, interpMode=RCIM_Cubic, leaveTangent=0.6186, tangentMode=RCTM_Break, time=0.1576, value=1.0645}
         {
          arriveTangent=-0.523
          arriveTangentWeight=0.3814
          interpMode=RCIM_Cubic
          leaveTangent=-0.523
          leaveTangentWeight=0.0762
          tangentMode=RCTM_User
          time=0.75
          value=1
         }
         {arriveTangent=-4, interpMode=RCIM_Cubic, leaveTangent=-4, time=0.8958, value=0.2748}
         {tangentMode=RCTM_User, time=1, value=0}
        ]
       }}}
     Scale Curve=0.92
    )
    Scale RGB=&MakeVectorFromLinearColorRGB(Color=&ColorFromCurve(ColorCurve=DataInterface() {propertyValues {
        AlphaCurve {keys []}
        BlueCurve {keys [<at 0 0.9809>,<at 0.058 2>,<at 0.1332 0.9601>,<at 1 0.374 RCIM_Cubic>]}
        GreenCurve {keys [<at 0 1>,<at 0.058 1.95>,<at 0.1332 1>,<at 1 0.374 RCIM_Cubic>]}
        RedCurve {keys [<at 0 0.6295>,<at 0.058 1.9478>,<at 0.1332 0.8525>,<at 1 0.374 RCIM_Cubic>]}
       }}))
    ScaleA=true
    ScaleRGB=true
   ) {module=ScaleColor001}
  ]
  rendererClasses [&NiagaraMeshRendererProperties]
  renderers [{rendererClass=&NiagaraMeshRendererProperties, rendererIndex=0}]
 }
 Ring_GLow {
  emitterUpdateScript [
   &EmitterState(Life Cycle Mode=::NewEnumerator1, Loop Behavior=::NewEnumerator1, Loop Duration=3, UseLoopDelay=false)
   &SpawnBurst_Instantaneous(Use Spawn Probability=false, UseLoopCountLimit=false)
   &SpawnBurst_Instantaneous(Spawn Group=1, Use Spawn Probability=false, UseLoopCountLimit=false) {module=SpawnBurst_Instantaneous001}
   /SetVariables(Emitter.Color:LinearColor=<rgba 0.1101 0.2135 0.2198 0.05>) {isSetParameters=true}
   /SetVariables(
    Emitter.Opacity Mult A:NiagaraFloat=&FloatFromCurve(
     FloatCurve=DataInterface() {propertyValues {Curve {
        defaultValue=1
        keys [
         {interpMode=RCIM_Cubic, tangentMode=RCTM_User, time=0.017, value=0}
         {interpMode=RCIM_Cubic, tangentMode=RCTM_User, time=0.0457, value=1}
         {tangentMode=RCTM_User, time=0.8717, value=0}
        ]
       }}}
     CurveIndex=$Emitter.NormalizedLoopAge
     Scale Curve=4
    )
    Emitter.Opacity Mult B:NiagaraFloat=&FloatFromCurve(
     FloatCurve=DataInterface() {propertyValues {Curve {
        defaultValue=1
        keys [
         {interpMode=RCIM_Cubic, tangentMode=RCTM_User, time=0.017, value=0}
         {interpMode=RCIM_Cubic, tangentMode=RCTM_User, time=0.0457, value=1}
         {tangentMode=RCTM_User, time=0.9, value=0}
        ]
       }}}
     CurveIndex=$Emitter.NormalizedLoopAge
     Scale Curve=6
    )
   ) {isSetParameters=true}
  ]
  particleSpawnScript [
   &InitializeParticle(
    Lifetime=2.65
    Mesh Scale=<1.2 1.2 1.2>
    Mesh Scale Mode=::NewEnumerator4
    Position Offset=&FN_FrontEndVector3Modifier(Gameplay=<0 0 0>, Front End=<-10 0 0>)
    Sprite Size Mode=::NewEnumerator3
    Uniform Sprite Size=300
    UsePositionOffset=true
    Write Material Random=true
   )
   &InitialMeshOrientation(
    Mesh Axis To Orient=<0 1 0>
    Mesh Orientation Mode=::NewEnumerator2
    Orientation Vector=<0 0 0>
    Rotation=<0 -0.25 0>
    Use Rotation Vector=true
   )
  ]
  particleUpdateScript [
   &ParticleState()
   &ScaleColor(
    Scale Alpha=&FloatFromCurve(FloatCurve=DataInterface() {propertyValues {Curve {
        defaultValue=1
        keys [
         {interpMode=RCIM_Cubic, tangentMode=RCTM_User, time=0.0418, value=0}
         {arriveTangent=2.5533, interpMode=RCIM_Cubic, leaveTangent=2.5533, time=0.2152, value=0.2935}
         {arriveTangentWeight=0.0595, interpMode=RCIM_Cubic, leaveTangentWeight=0.1725, tangentMode=RCTM_User, time=0.4242, value=0.9765}
         {arriveTangentWeight=0.2718, interpMode=RCIM_Cubic, leaveTangentWeight=0.1596, tangentMode=RCTM_User, time=0.6962, value=1}
         {tangentMode=RCTM_User, time=0.9805, value=0}
        ]
       }}})
    ScaleA=true
    ScaleRGB=true
   )
  ]
  rendererClasses [&NiagaraMeshRendererProperties]
  renderers [{rendererClass=&NiagaraMeshRendererProperties, rendererIndex=0}]
 }
 SmokeRing001 {
  emitterUpdateScript [
   &EmitterState(Life Cycle Mode=::NewEnumerator1, Loop Behavior=::NewEnumerator1, Loop Delay=3.4, Loop Duration=3.2, UseLoopDelay=true)
   &SpawnBurst_Instantaneous(Spawn Count=15, Use Spawn Probability=false, UseLoopCountLimit=false)
   &SpawnRate(
    SpawnRate=&FloatFromCurve(
     FloatCurve=DataInterface() {propertyValues {Curve {
        defaultValue=1
        keys [
         {interpMode=RCIM_Cubic, tangentMode=RCTM_User, time=0.2, value=1}
         {
          arriveTangent=-1.8776
          arriveTangentWeight=0.4224
          interpMode=RCIM_Cubic
          leaveTangent=-1.8776
          leaveTangentWeight=0.049
          tangentMode=RCTM_User
          time=0.7016
          value=0.4094
         }
         {arriveTangent=0.0296, interpMode=RCIM_Cubic, leaveTangent=0.0296, tangentMode=RCTM_User, time=0.9948, value=0}
        ]
       }}}
     CurveIndex=$Emitter.NormalizedLoopAge
     Scale Curve=18
    )
    Use Spawn Probability=false
   )
  ]
  particleSpawnScript [
   &InitializeParticle(
    Lifetime Max=0.9
    Lifetime Min=0.4
    Lifetime Mode=::NewEnumerator1
    Mesh Scale Max=<1.65 1.65 1.8>
    Mesh Scale Min=<1.6 1.6 1.6>
    Mesh Scale Mode=::NewEnumerator2
    Position Offset=&FN_FrontEndVector3Modifier(Gameplay=<0 0 0>, Front End=<-10 0 0>)
    UsePositionOffset=true
    Write Material Random=true
   )
   &ShapeLocation(
    Cylinder Height=2
    Cylinder Height Midpoint=1
    Cylinder Radius=5
    MaskSpawnGroup=false
    Offset=<0 0 5>
    Offset Mode=::NewEnumerator0
    Override Seed=false
    Shape Primitive=::NewEnumerator1
    Surface Only=false
    UseOwnerScale=true
   )
   &InitialMeshOrientation(
    Mesh Axis To Orient=<0 1 0>
    Mesh Orientation Mode=::NewEnumerator2
    Orientation Vector=<0 0 1>
    Rotation=&RandomRangeVector(Minimum=<0 0 0>, Maximum=<0 0 2>)
    Use Rotation Vector=true
   )
   &AddVelocity(Velocity=&RandomRangeVector(Minimum=<0 0 15>, Maximum=<0 0 60>)) {enabled=false}
  ]
  particleUpdateScript [
   &ParticleState()
   &UpdateMeshOrientation(
    Rotation Rate=&Multiply_Float(
     A=&Multiply_Float(
      A=&RandomRangeFloat(Minimum=-0.4, Maximum=0.4)
      B=&FloatFromCurve(FloatCurve=DataInterface() {propertyValues {Curve {keys [
           {interpMode=RCIM_Cubic, tangentMode=RCTM_User, time=0, value=1}
           {arriveTangent=-3.5121, interpMode=RCIM_Cubic, leaveTangent=-3.5121, tangentMode=RCTM_User, time=1.0041, value=0}
          ]}}})
     )
     B=&FloatFromCurve(
      FloatCurve=DataInterface() {propertyValues {Curve {keys [
          {interpMode=RCIM_Cubic, tangentMode=RCTM_User, time=-0.0002, value=0.9548}
          {interpMode=RCIM_Cubic, tangentMode=RCTM_User, time=0.0456, value=1.2227}
          {arriveTangent=-2.0696, interpMode=RCIM_Cubic, leaveTangent=-2.0696, tangentMode=RCTM_User, time=0.1952, value=0.5}
          {
           arriveTangent=-0.511
           arriveTangentWeight=0.0848
           interpMode=RCIM_Cubic
           leaveTangent=-0.511
           leaveTangentWeight=0.2422
           tangentMode=RCTM_User
           time=0.5576
           value=0.1228
          }
          {arriveTangent=-0.0256, interpMode=RCIM_Cubic, leaveTangent=-0.0256, tangentMode=RCTM_User, time=0.9913, value=0}
         ]}}}
      CurveIndex=$Emitter.NormalizedLoopAge
     )
    )
    Rotation Vector=<0 0 1>
   )
   &ScaleMeshSize(Scale Factor=&VectorFromFloat(Value=&FloatFromCurve(
      FloatCurve=DataInterface() {propertyValues {Curve {
         defaultValue=1
         keys [<at -0.0078 1 RCIM_Cubic>, {arriveTangent=-0.3116, interpMode=RCIM_Cubic, leaveTangent=-0.3116, tangentMode=RCTM_User, time=1, value=1.5}]
        }}}
      Scale Curve=&RandomRangeFloat(Minimum=1, Maximum=1.25)
     )))
   &ScaleColor(
    Scale Alpha=&Multiply_Float(
     A=&FloatFromCurve(
      FloatCurve=DataInterface() {propertyValues {Curve {
         defaultValue=1
         keys [
          {interpMode=RCIM_Cubic, tangentMode=RCTM_User, time=0, value=0}
          {interpMode=RCIM_Cubic, tangentMode=RCTM_User, time=0.1617, value=0.6549}
          {tangentMode=RCTM_User, time=1, value=0}
         ]
        }}}
      Scale Curve=0.12
     )
     B=&FloatFromCurve(
      FloatCurve=DataInterface() {propertyValues {Curve {
         defaultValue=1
         keys [
          <at 0.6866 1 RCIM_Cubic>
          {arriveTangent=-7.3431, interpMode=RCIM_Cubic, leaveTangent=-7.3431, tangentMode=RCTM_User, time=0.7817, value=0.2843}
          {arriveTangent=-0.6899, interpMode=RCIM_Cubic, leaveTangent=-0.6899, tangentMode=RCTM_User, time=0.85, value=0}
         ]
        }}}
      CurveIndex=$Emitter.NormalizedLoopAge
     )
    )
    Scale RGB=&MakeVectorFromLinearColorRGB(Color=/Niagara/DynamicInputs/Add/Add_LinearColor(
      A:LinearColor=&ColorFromCurve(ColorCurve=DataInterface() {propertyValues {
         AlphaCurve {keys [<at 0 0>,<at 0.0901 1 RCIM_Cubic>,<at 0.3915 0.5>,<at 1 0 RCIM_Cubic>]}
         BlueCurve {keys [<at 0.009 1>,<at 0.0499 0.9927>,<at 0.2996 1>,<at 1 0.1965 RCIM_Cubic>]}
         GreenCurve {keys [<at 0.009 0.8864>,<at 0.0499 0.974>,<at 0.2996 0.89>,<at 1 0.2031 RCIM_Cubic>]}
         RedCurve {keys [<at 0.009 0.4609>,<at 0.0499 0.7877>,<at 0.2996 0.4595>,<at 1 0.079 RCIM_Cubic>]}
        }})
      B:LinearColor=/Niagara/DynamicInputs/LinearColor/MakeLinearColorFromHSV(
       HSV From:NiagaraInt32=1
       HSV:Vector3f=?
       Hue:NiagaraFloat=0
       Saturation:NiagaraFloat=0
       Value:NiagaraFloat=&FloatFromCurve(
        FloatCurve=DataInterface() {propertyValues {Curve {keys [
            {arriveTangent=60.3451, interpMode=RCIM_Cubic, leaveTangent=60.3451, tangentMode=RCTM_User, time=-0.0007, value=1}
            {arriveTangent=-14.8494, interpMode=RCIM_Cubic, leaveTangent=-14.8494, tangentMode=RCTM_User, time=0.0405, value=0.3673}
            {interpMode=RCIM_Cubic, tangentMode=RCTM_User, time=0.1045, value=0}
           ]}}}
        CurveIndex=$Emitter.NormalizedLoopAge
       )
       Alpha:NiagaraFloat=0
      )
     ))
    ScaleA=true
    ScaleRGB=true
   )
   &PointForce(
    Force Origin=&AddVectorToPosition(Position=/Niagara/DynamicInputs/Helpers/SimulationPosition(), Vector=<0 0 50>)
    Force Strength=&RandomRangeFloat(Minimum=0, Maximum=30)
    Module.Use Falloff=false
    Module.UseLinearForce=false
   )
   &SolveForcesAndVelocity(Clamp Velocity=false, Limit Acceleration=false)
  ]
  rendererClasses [&NiagaraMeshRendererProperties]
  renderers [{rendererClass=&NiagaraMeshRendererProperties, rendererIndex=0}]
 }
 SmokeRing002 {
  emitterUpdateScript [
   &EmitterState(Life Cycle Mode=::NewEnumerator1, Loop Behavior=::NewEnumerator1, Loop Delay=0.62, Loop Duration=6.6, UseLoopDelay=true)
   &SpawnRate(
    SpawnRate=&FloatFromCurve(
     FloatCurve=DataInterface() {propertyValues {Curve {
        defaultValue=1
        keys [
         {arriveTangent=20.6992, interpMode=RCIM_Cubic, leaveTangent=20.6992, tangentMode=RCTM_User, time=0.1963, value=0.3}
         {interpMode=RCIM_Cubic, tangentMode=RCTM_User, time=0.2981, value=0.9882}
         {arriveTangent=-1.5376, interpMode=RCIM_Cubic, leaveTangent=-1.5376, time=0.6514, value=0.5704}
         {arriveTangent=0.0296, interpMode=RCIM_Cubic, leaveTangent=0.0296, tangentMode=RCTM_User, time=0.9408, value=0}
        ]
       }}}
     CurveIndex=$Emitter.NormalizedLoopAge
     Scale Curve=15
    )
    Use Spawn Probability=false
   )
  ]
  particleSpawnScript [
   &InitializeParticle(
    Color=<rgba 2.1107 2.5428 3 0.2>
    Lifetime Max=0.9
    Lifetime Min=0.5
    Lifetime Mode=::NewEnumerator1
    Mesh Scale Max=<1.75 1.75 2.1>
    Mesh Scale Min=<1.7 1.7 1.7>
    Mesh Scale Mode=::NewEnumerator2
    Position Offset=&FN_FrontEndVector3Modifier(Gameplay=<0 0 0>, Front End=<-10 0 0>)
    UsePositionOffset=true
    Write Material Random=true
   )
   &ShapeLocation(
    Cylinder Height=1
    Cylinder Height Midpoint=1
    Cylinder Radius=1
    MaskSpawnGroup=false
    Offset=<0 0 5>
    Offset Mode=::NewEnumerator0
    Override Seed=false
    Shape Primitive=::NewEnumerator1
    Surface Only=false
    UseOwnerScale=true
   ) {enabled=false}
   &InitialMeshOrientation(
    Mesh Axis To Orient=<0 1 0>
    Mesh Orientation Mode=::NewEnumerator2
    Orientation Vector=<0 0 1>
    Rotation=&RandomRangeVector(Minimum=<0 0 0>, Maximum=<0 0 2>)
    Use Rotation Vector=true
   )
   &AddVelocity(Velocity=&RandomRangeVector(Minimum=<0 0 15>, Maximum=<0 0 60>)) {enabled=false}
  ]
  particleUpdateScript [
   &ParticleState()
   &ScaleMeshSize(Scale Factor=&Multiply_Vector(
     A=&VectorFromFloat(Value=&FloatFromCurve(FloatCurve=DataInterface() {propertyValues {Curve {keys [<at 0.0046 0.2824>,<at 0.5025 0.9576>]}}}, CurveIndex=$Emitter.NormalizedLoopAge))
     B=&VectorFromFloat(Value=&FloatFromCurve(
       FloatCurve=DataInterface() {propertyValues {Curve {
          defaultValue=1
          keys [
           {interpMode=RCIM_Cubic, tangentMode=RCTM_User, time=0, value=1}
           {arriveTangent=0.3803, interpMode=RCIM_Cubic, leaveTangent=0.3803, tangentMode=RCTM_User, time=1, value=1.1}
          ]
         }}}
       Scale Curve=&RandomRangeFloat(Minimum=1, Maximum=1.2)
      ))
    ))
   &UpdateMeshOrientation(
    Rotation Rate=&Multiply_Float(
     A=&Multiply_Float(
      A=&RandomRangeFloat(Minimum=-0.05, Maximum=0.05)
      B=&FloatFromCurve(FloatCurve=DataInterface() {propertyValues {Curve {keys [
           {interpMode=RCIM_Cubic, tangentMode=RCTM_User, time=0, value=1}
           {arriveTangent=-3.5121, interpMode=RCIM_Cubic, leaveTangent=-3.5121, tangentMode=RCTM_User, time=1.0041, value=0}
          ]}}})
     )
     B=&FloatFromCurve(
      FloatCurve=DataInterface() {propertyValues {Curve {keys [
          {interpMode=RCIM_Cubic, tangentMode=RCTM_User, time=0, value=1}
          {interpMode=RCIM_Cubic, tangentMode=RCTM_User, time=0.0458, value=1.2678}
          {arriveTangent=-1.2911, interpMode=RCIM_Cubic, leaveTangent=-1.2911, tangentMode=RCTM_User, time=0.1562, value=1}
          {
           arriveTangent=-0.3731
           arriveTangentWeight=0.0848
           interpMode=RCIM_Cubic
           leaveTangent=-0.3731
           leaveTangentWeight=0.2422
           tangentMode=RCTM_User
           time=0.4828
           value=0.6758
          }
          {arriveTangent=-0.0256, interpMode=RCIM_Cubic, leaveTangent=-0.0256, tangentMode=RCTM_User, time=0.9951, value=0.5}
         ]}}}
      CurveIndex=$Emitter.NormalizedLoopAge
     )
    )
    Rotation Vector=<0 0 1>
   )
   &PointForce(
    Force Origin=&AddVectorToPosition(Position=&SimulationPosition(), Vector=<0 0 0>)
    Force Strength=&RandomRangeFloat(Minimum=0, Maximum=15)
    Module.Use Falloff=false
    Module.UseLinearForce=false
   )
   &SolveForcesAndVelocity(Clamp Velocity=false, Limit Acceleration=false)
   &ScaleColor(
    Scale Alpha=&Multiply_Float(
     A=&FloatFromCurve(
      FloatCurve=DataInterface() {propertyValues {Curve {
         defaultValue=1
         keys [
          {interpMode=RCIM_Cubic, tangentMode=RCTM_User, time=0, value=0}
          {arriveTangentWeight=0.1532, interpMode=RCIM_Cubic, leaveTangentWeight=0.0844, tangentMode=RCTM_User, time=0.4595, value=1.0196}
          {
           arriveTangent=-2.3085
           arriveTangentWeight=0.1801
           interpMode=RCIM_Cubic
           leaveTangent=-2.3085
           leaveTangentWeight=0.2046
           tangentMode=RCTM_User
           time=0.6975
           value=0.3163
          }
          {tangentMode=RCTM_User, time=1, value=0}
         ]
        }}}
      CurveIndex=$Emitter.NormalizedLoopAge
      Scale Curve=0.5
     )
     B=&FloatFromCurve(FloatCurve=DataInterface() {propertyValues {Curve {
         defaultValue=1
         keys [
          {interpMode=RCIM_Cubic, tangentMode=RCTM_User, time=0, value=0}
          {interpMode=RCIM_Cubic, tangentMode=RCTM_User, time=0.5, value=1}
          {tangentMode=RCTM_User, time=1, value=0}
         ]
        }}})
    )
    ScaleA=true
    ScaleRGB=true
   )
   &ScaleColor(
    Scale Alpha=&Multiply_Float(
     A=&FloatFromCurve(
      FloatCurve=DataInterface() {propertyValues {Curve {
         defaultValue=1
         keys [
          {interpMode=RCIM_Cubic, tangentMode=RCTM_User, time=0, value=0}
          {arriveTangentWeight=0.1532, interpMode=RCIM_Cubic, leaveTangentWeight=0.0844, tangentMode=RCTM_User, time=0.4595, value=1.0196}
          {arriveTangentWeight=0.1801, interpMode=RCIM_Cubic, leaveTangentWeight=0.2046, tangentMode=RCTM_User, time=0.7125, value=1}
          {tangentMode=RCTM_User, time=1, value=0}
         ]
        }}}
      CurveIndex=$Emitter.NormalizedLoopAge
      Scale Curve=0.5
     )
     B=&RemapRange(
      Input Value To Remap=&Abs_Float(Float:NiagaraFloat=&DotProduct(
        Vector A=&VectorFromCurve(
         VectorCurve=DataInterface() {propertyValues {XCurve {keys [<at 0 0>,<at 1 1>]}, YCurve {keys [<at 0.7 1>,<at 1 0>]}, ZCurve {keys [<at 0.7 1>,<at 1 0>]}}}
         CurveIndex=&Abs_Float(Float:NiagaraFloat=&MakeFloatFromVector(
           Vector=&ReturnCameraVector(Camera Vectors=::NewEnumerator1, Coordinate Space=::Local, CameraQuery=DataInterface() {propertyValues {bRequireCurrentFrameData=false}})
           Channel=::NewEnumerator0
          ))
         Scale Curve=<1 0 0>
        )
        Vector B=<1 0 0>
       ))
      Input Min=0.2
      Input Max=0.9
      Output Min=0
      Output Max=1
     )
    )
    ScaleA=true
    ScaleRGB=true
   ) {enabled=false, module=ScaleColor001}
  ]
  rendererClasses [&NiagaraMeshRendererProperties]
  renderers [{rendererClass=&NiagaraMeshRendererProperties, rendererIndex=0}]
 }
 Stars001 {
  emitterUpdateScript [
   &EmitterState(Life Cycle Mode=::NewEnumerator1, Loop Behavior=::NewEnumerator1, Loop Delay=3, UseLoopDelay=true)
   &SpawnBurst_Instantaneous(Spawn Count=25, Use Spawn Probability=false, UseLoopCountLimit=false)
  ]
  particleSpawnScript [
   &InitializeParticle(
    Color Maximum=<rgba 1.1386 1.6579 2 1>
    Color Minimum=<rgba 0.3945 0.6791 0.7 1>
    Color Mode=::NewEnumerator2
    Lifetime Max=5
    Lifetime Min=2
    Lifetime Mode=::NewEnumerator1
    Position Offset=&FN_FrontEndVector3Modifier(Gameplay=<0 0 0>, Front End=<-10 0 0>)
    Sprite Rotation Mode=::NewEnumerator2
    Sprite Size Mode=::NewEnumerator1
    Sprite UV Mode=::NewEnumerator3
    Uniform Sprite Size Max=12
    Uniform Sprite Size Min=1
    UsePositionOffset=true
    Write Material Random=true
   )
   &ShapeLocation(
    MaskSpawnGroup=false
    Override Seed=false
    Ring Radius=80
    Rotation Axis=<0 1 0>
    Rotation Mode=::NewEnumerator1
    Shape Primitive=::NewEnumerator4
    UseOwnerScale=true
   )
   &AddVelocity(
    Module.Use Falloff Distance=false
    Override Seed=false
    Velocity Mode=::NewEnumerator1
    Velocity Speed=&RandomRangeFloat(Minimum=25, Maximum=200)
   )
   &ApplyInitialForces()
   &CameraOffset(Camera Offset Amount=&RandomRangeFloat(Minimum=-5, Maximum=15))
  ]
  particleUpdateScript [
   &ParticleState()
   &CurlNoiseForce(Bias Noise=false, Noise Frequency=20, Noise Strength=20, Pan Noise=false)
   &ScaleSpriteSize(Scale Factor=&Vector2DFromFloat(Value=&FloatFromCurve(FloatCurve=DataInterface() {propertyValues {Curve {keys [
          <at 0 0 RCIM_Cubic>
          <at 0.112 1.0173>
          {arriveTangent=-1.142, interpMode=RCIM_Cubic, leaveTangent=-1.142, time=0.7001, value=0.872}
          <at 1.0011 0.002>
         ]}}})))
   &ScaleColor(
    Scale Alpha=&FloatFromCurve(FloatCurve=DataInterface() {propertyValues {Curve {keys [<at 0 0 RCIM_Cubic>,<at 0.5 1 RCIM_Cubic>,<at 1 0 RCIM_Cubic>]}}})
    ScaleA=true
    ScaleRGB=true
   )
   &Drag(Drag=&RandomRangeFloat(Minimum=2.5, Maximum=5), Use Linear Drag=true, Use Rotational Drag=true)
   &PointForce(
    Force Origin=&AddVectorToPosition(Position=&SimulationPosition(), Vector=<0 0 0>)
    Force Strength=10
    Module.Use Falloff=false
    Module.UseLinearForce=false
   )
   &SolveForcesAndVelocity(Clamp Velocity=false, Limit Acceleration=false)
  ]
  rendererClasses [&NiagaraSpriteRendererProperties]
  renderers [{rendererClass=&NiagaraSpriteRendererProperties}]
 }
}
summary {
 systemName=NS_Emote_GrimHound_Moon
 emitters [
  {emitterName=Moon001, rendererClasses [&NiagaraMeshRendererProperties]}
  {emitterName=Stars001, rendererClasses [&NiagaraSpriteRendererProperties]}
  {emitterName=SmokeRing001, rendererClasses [&NiagaraMeshRendererProperties]}
  {emitterName=Ink_Circle, rendererClasses [&NiagaraSpriteRendererProperties]}
  {emitterName=Mesh_Swirls, rendererClasses [&NiagaraMeshRendererProperties]}
  {emitterName=Ring_GLow, rendererClasses [&NiagaraMeshRendererProperties]}
  {emitterName=Glow002, rendererClasses [&NiagaraSpriteRendererProperties]}
  {emitterName=SmokeRing002, rendererClasses [&NiagaraMeshRendererProperties]}
 ]
 userVariables [
  User.bIsFrontEnd {type=NiagaraBool}
  User.bIsFrontEndPreview {type=NiagaraBool}
 ]
}
