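# Niagara ribbon system that draws translucent teal animation trails for the main player's GrimHound emote.
# A single AnimTrails emitter uses FN anim-trail modules to spawn, initialize bone indices and update particles against a system-level skeletal mesh.
# Line attraction adds a force at spawn and in-update solving moves the particles.
# A width curve drops steeply then tapers to zero, and a paler, fainter secondary color is set each update.
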NiagaraSystem /BRCosmetics/Animation/Game/MainPlayer/Emotes/GrimHound/VFX/NS_Emote_GrimHound_Trails
dependencies {
 usedDataInterfaces [&NiagaraDataInterfaceCurve]
 usedDynamicInputs []
 usedModules [
  &SystemState
  &self:SystemSpawnScript.SystemScriptSource.SystemScriptGraph.NiagaraNodeAssignment_1.SetVariables_#1
 ]
}
emitters {AnimTrails {
  emitterUpdateScript [
   &EmitterState(Inactive Response=::NewEnumerator1, Life Cycle Mode=::NewEnumerator1, Loop Duration=1, UseLoopDelay=false)
   &FN_SpawnAnimTrail()
  ]
  particleSpawnScript [
   &FN_InitializeAnimTrailBoneIndecies()
   &FN_InitializeAnimTrail(AnimTrailWidth=3, Color=<rgba 0.1597 0.6496 0.7031 0.65>, SkeletalMesh=$System.Skeletal Mesh)
   /Niagara/Modules/Update/Forces/LineAttractionForce(Attraction Strength=100, Line End:Vector3f=<0 100 0>, Line Start:Vector3f=<0 0 0>, Use Attraction Falloff=false)
   &ApplyInitialForces()
  ]
  particleUpdateScript [
   &FN_UpdateAnimTrail(SkeletalMesh=$System.Skeletal Mesh) {module=FN_UpdateAnimTrail001}
   &ScaleRibbonWidth(Ribbon Width Scale=&FloatFromCurve(FloatCurve=DataInterface() {propertyValues {Curve {keys [
         {arriveTangent=-14.2285, interpMode=RCIM_Cubic, leaveTangent=-14.2285, tangentMode=RCTM_User, time=0.02, value=1}
         {arriveTangent=-1.0312, interpMode=RCIM_Cubic, leaveTangent=-1.4131, tangentMode=RCTM_Break, time=0.1039, value=0.3893}
         {arriveTangent=-0.5682, interpMode=RCIM_Cubic, leaveTangent=-0.5682, tangentMode=RCTM_User, time=0.2706, value=0.1849}
         {arriveTangent=-0.152, interpMode=RCIM_Cubic, leaveTangent=-0.152, tangentMode=RCTM_User, time=1, value=0}
        ]}}}))
   &SolveForcesAndVelocity(Clamp Velocity=false, Limit Acceleration=false)
   /SetVariables(Particles.Second Color:LinearColor=<rgba 0.2734 0.5052 0.4615 0.3>) {isSetParameters=true}
  ]
  rendererClasses [&NiagaraRibbonRendererProperties]
  renderers [{rendererClass=&NiagaraRibbonRendererProperties}]
 }}
summary {
 systemName=NS_Emote_GrimHound_Trails
 emitters [{emitterName=AnimTrails, rendererClasses [&NiagaraRibbonRendererProperties]}]
 userVariables [
  User.NewVector3f {type=Vector3f, description="", defaultValue {struct=Vector3f, value=<10 0 0>}}
  User.RibbonIDOffset {type=NiagaraInt32}
  User.SkeletalMesh {
   type=NiagaraDataInterfaceSkeletalMesh
   defaultValue {
    struct=NiagaraExt_VariableValue_DataInterface
    value {dataInterface=&self:NiagaraDataInterfaceSkeletalMesh_3}
   }
  }
  User.TrailLength {type=NiagaraFloat, defaultValue=0, description=""}
 ]
}
